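# Animation Blueprint for the F_MED_FearCatch female player body, layering a Control Rig over the pose copied from the attached parent mesh.
# The update event splits PlayerMovement output into clamped directional floats, while gameplay events set tactical sprint, jump (excluding skydiving) and equipped-weapon flags.
# The rig receives forward/backward, jump, crouch, sprint, parachute and tactical shotgun state (stored as isTacticalPistol Equipped); strafe and other weapon flags are not passed on.
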AnimBlueprint /BRCosmetics/Characters/Player/Female/Medium/Bodies/F_MED_FearCatch/Meshes/F_MED_FearCatch_AnimBP
target_skeleton=&/F_MED_FearCatch_Skeleton
graphs [
 AnimGraph {nodes [
   Output Pose(Result=Use cached pose 'CR_Cache_Input'()) {layerGroup=None}
   n1 = Control Rig!(
    Source=Copy Pose From Mesh() {bUseAttachedParent=true}
    Alpha
    IsJumping:bool=Get IsJumping() -> (IsJumping:bool)
    Pub_Is_Crouching:bool=Get bIsCrouching()
    Pub_Is_TacticalSprint:double=Get IsTacticalSprint() -> (IsTacticalSprint:double)
    Pub_Is_Sprinting:bool=Get bIsSprinting()
    Pub_Is_Forward:double=Get moveForward() -> (moveForward:double)
    Pub_Is_Backward:double=Get moveBackward() -> (moveBackward:double)
    Pub_Is_TactialShotgun:double=Get isTacticalPistol Equipped() -> (isTacticalPistol Equipped:double)
    Pub_Is_Parachute:bool=Get bIsParachuteOpened()
   ) -> (Pose) {controlRigAssetReference {blueprintRigClass=&/F_MED_FearCatch_CtrlRig.F_MED_FearCatch_CtrlRig_C}}
   Save cached pose 'CR_Cache_Input'(Pose=@n1) {cacheName=CR_Cache_Input}
  ]}
 EventGraph {nodes [
   Event BlueprintUpdateAnimation()
   > Sequence()
   > n1 = PlayerMovement(AnimInstance=Self-Reference(), XDotProductCoefficient=1, YDotProductCoefficient=1, ZDotProductCoefficient=1)
   n2 = BreakVector(InVec=MakeVector(X=@n1, Y=@n1.YMovement))
   > Set strafeLeft float(strafeLeft float:double="Absolute (Float)"(A="Clamp (Float)"(Value=@n2, Min=-1, Max=0))) -> (Output_Get:double)
   > Set strafeRight float(strafeRight float:double="Clamp (Float)"(Value=@n2, Min=0, Max=1)) -> (Output_Get:double)
   > Set moveForward(moveForward:double="Clamp (Float)"(Value=@n2.Y, Min=0, Max=1)) -> (Output_Get:double)
   > Set moveBackward(moveBackward:double="Absolute (Float)"(A="Clamp (Float)"(Value=@n2.Y, Min=-1, Max=0))) -> (Output_Get:double)
   Event OnBeginTacticalSprinting() {is=title K2Node_Event}
   > Set IsTacticalSprint(IsTacticalSprint:double=1) -> (Output_Get:double)
   Event OnEndTacticalSprinting() {is=title K2Node_Event}
   > Set IsTacticalSprint(IsTacticalSprint=0)
   Event OnWeaponChanged()
   > Set isAR Equipped(isAR Equipped:double=0) -> (Output_Get:double)
   > Set isPistol Equipped(isPistol Equipped:double=0) -> (Output_Get:double)
   > Set isTacticalPistol Equipped(isTacticalPistol Equipped:double=0) -> (Output_Get:double)
   > Set isSMGEquipped(isSMGEquipped:double=0) -> (Output_Get:double)
   > n3 = Switch on EFortWeaponCoreAnimation!(execute, Selection:EFortWeaponCoreAnimation Enum=Get WeaponCoreAnim!() -> (WeaponCoreAnim)) -> (
    Melee:Exec
    Pistol:Exec
    Shotgun:Exec
    PaperBlueprint:Exec
    Rifle:Exec
    MeleeOneHand:Exec
    MachinePistol:Exec
    RocketLauncher:Exec
    GrenadeLauncher:Exec
    GoingCommando:Exec
    AssaultRifle:Exec
    TacticalShotgun:Exec
    SniperRifle:Exec
    TrapPlacement:Exec
    ShoulderLauncher:Exec
    AbilityDecoTool:Exec
    Crossbow:Exec
    C4:Exec
    RemoteControl:Exec
    DualWield:Exec
    AR_BullPup:Exec
    AR_ForwardGrip:Exec
    MedPackPaddles:Exec
    SMG_P90:Exec
    AR_DrumGun:Exec
    Consumable_Small:Exec
    Consumable_Large:Exec
    Balloon:Exec
    MountedTurret:Exec
    CreativeTool:Exec
    ExplosiveBow:Exec
    AshtonIndigo:Exec
    AshtonChicago:Exec
    MeleeDualWield:Exec
    Unarmed:Exec
   )
   Set isAR Equipped(execute=@n3.AssaultRifle, isAR Equipped=1)
   Set isPistol Equipped(execute=@n3.Pistol, isPistol Equipped=1)
   Set isSMGEquipped(execute=@n3.SMG_P90, isSMGEquipped=1)
   Set isTacticalPistol Equipped(execute=@n3.TacticalShotgun, isTacticalPistol Equipped=1)
   Event OnEndFalling()
   > Set IsJumping(IsJumping:bool=false) -> (Output_Get:bool)
   Event OnBeginJumping()
   n4 = Get bIsSkydiveFloating()
   n5 = Get bIsSkyDiving()
   > Set IsJumping(IsJumping=NOT Boolean(A=OR Boolean(A=@n4, B=@n5)))
  ]}
]
variables [
 Gravity Override Params {
  type=FortGravityOverrideParameterSet
  default=(Skydive_GoingUp=(GravityOverride=(WindFrequency=(X=1.500000,Y=7.000000,Z=8.000000),WindAmplitude=(X=200.000000,Y=-500.000000,Z=400.000000),GravityOverride=(X=0.000000,Y=-300.000000,Z=1500.000000),JointName="root")),Skydive_FloatingDown=(GravityOverride=(WindFrequency=(X=1.500000,Y=7.000000,Z=8.000000),WindAmplitude=(X=200.000000,Y=-500.000000,Z=400.000000),WindOffset=(X=0.000000,Y=-600.000000,Z=400.000000),GravityOverride=(X=0.000000,Y=-700.000000,Z=600.000000),JointName="root")),Skydive_DivingDown=(GravityOverride=(WindFrequency=(X=1.500000,Y=12.000000,Z=10.000000),WindAmplitude=(X=50.000000,Y=300.000000,Z=100.000000),GravityOverride=(X=-1500.000000,Y=0.000000,Z=0.000000),JointName="pelvis")),Skydive_Parachuting=(GravityOverride=(WindFrequency=(X=1.500000,Y=7.000000,Z=8.000000),WindAmplitude=(X=200.000000,Y=-500.000000,Z=400.000000),GravityOverride=(X=0.000000,Y=-800.000000,Z=-980.000000),JointName="root")),OnGround_Standing=(GravityOverride=(GravityOverride=(X=0.000000,Y=0.000000,Z=-980.000000),JointName="root")),OnGround_Moving=(GravityOverride=(GravityOverride=(X=0.000000,Y=0.000000,Z=-980.000000),JointName="root")),OnGround_Moving_HighRotationRate=(GravityOverride=(GravityOverride=(X=0.000000,Y=0.000000,Z=-980.000000),JointName="root")),OnGround_Crouch_Moving=(GravityOverride=(GravityOverride=(X=0.000000,Y=0.000000,Z=-980.000000),JointName="root")),Emote_or_Melee=(GravityOverride=(GravityOverride=(X=0.000000,Y=0.000000,Z=-980.000000),JointName="root")),Swimming=(GravityOverride=(GravityOverride=(X=-800.000000,Y=0.000000,Z=0.000000),JointName="pelvis")),Front_End=(GravityOverride=(GravityOverride=(X=0.000000,Y=0.000000,Z=-980.000000),JointName="root")),DBNO=(GravityOverride=(GravityOverride=(X=-980.000000,Y=400.000000,Z=0.000000),JointName="pelvis")),DBNO_Carried=(GravityOverride=(GravityOverride=(X=-980.000000,Y=500.000000,Z=-500.000000),JointName="pelvis")),OnGround_Falling=(GravityOverride=(GravityOverride=(X=0.000000,Y=0.000000,Z=-980.000000),JointName="root")),Ziplining=(GravityOverride=(GravityOverride=(X=0.000000,Y=0.000000,Z=-980.000000),JointName="root")),Skydive_Surfing=(GravityOverride=(WindFrequency=(X=1.500000,Y=7.000000,Z=8.000000),WindAmplitude=(X=200.000000,Y=-500.000000,Z=400.000000),GravityOverride=(X=500.000000,Y=-300.000000,Z=-980.000000),JointName="root")),Riding_Standing=(GravityOverride=(GravityOverride=(X=0.000000,Y=-350.000000,Z=-980.000000),JointName="root")),Riding_Moving=(GravityOverride=(GravityOverride=(X=0.000000,Y=-450.000000,Z=-450.000000),JointName="root")),Riding_Moving_Targeting=(GravityOverride=(GravityOverride=(X=0.000000,Y=-200.000000,Z=-980.000000),JointName="root")),Tactical_Sprint=(GravityOverride=(GravityOverride=(X=0.000000,Y=-450.000000,Z=-500.000000),JointName="root")),Rail_Grinding=(GravityOverride=(GravityOverride=(X=0.000000,Y=0.000000,Z=-980.000000),JointName="root")),)
 }
 GravityOverride {type=Vector}
 isAR Equipped {type=number}
 IsJumping {type=boolean}
 isPistol Equipped {type=number}
 isSMGEquipped {type=number}
 isTacticalPistol Equipped {type=number}
 IsTacticalSprint {type=number}
 moveBackward {type=number}
 moveForward {type=number}
 strafeLeft float {type=number}
 strafeRight float {type=number}
]
